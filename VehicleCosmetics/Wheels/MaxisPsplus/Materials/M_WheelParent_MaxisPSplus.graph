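# Unlit parent material for the skeletal-mesh Maxis PSplus vehicle wheel cosmetic, with a cel-shaded look.
# Static switches choose textured, custom-color or RGB-mask-palette coloring for rim and tire, apparently blended by red vertex color.
# Stepped camera-vector-to-normal dot products supply the shading multiplier.
# Optional switches add eye jitter via alpha-masked, panning-gradient UV offsets, sine/cosine world-position wobble, and a timed vertex-color facial state whose lerp reads red for both inputs.
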Material /VehicleCosmetics/Wheels/MaxisPsplus/Materials/M_WheelParent_MaxisPSplus
graph {
 properties {
  bAutomaticallySetUsageInEditor=false
  bUsedWithSkeletalMesh=true
  previewMesh=/VehicleCosmetics/Wheels/MaxisPsplus/Meshes/SK_Wheel_Maxis_PSPlus
  shadingModel=MSM_Unlit
  usedShadingModels=Unlit
 }
 nodes [
  n1 = TextureCoordinate() {uTiling=0, vTiling=0}
  n2 = Constant() {r=1200}
  n3 = Multiply(
   A=TextureSample(UVs=Panner(Coordinate=@n1) {speedX=0.2, speedY=0.2}) {texture=/VehicleCosmetics/Wheels/_SharedTextures/FX/MowGradient01}.R
   B=@n2
  )
  n4 = Multiply(
   A=TextureSample(UVs=Panner(Coordinate=@n1) {speedX=-0.1, speedY=-0.1}) {texture=&MowGradient01}.R
   B=@n2
  )
  n5 = ConstantBiasScale(None=AppendVector(A=Clamp(None=@n3), B=Clamp(None=@n4))) {bias=-0.5, scale=0.0175}
  n6 = Multiply(A=TextureSampleParameter2D() {parameterName=BC_Tire, texture=/VehicleCosmetics/Wheels/MaxisPsplus/Textures/T_Maxis_PSPlus_Tire_BC}.A, B=@n5)
  n7 = TextureCoordinate()
  n8 = TextureSampleParameter2D(UVs=StaticSwitchParameter(True=Add(A=@n6, B=@n7), False=@n7) {parameterName=Eye}) {parameterName=BC_Tire, texture=&T_Maxis_PSPlus_Tire_BC}
  n9 = NamedRerouteUsage()
  n10 = LinearInterpolate(
   A=Multiply(A=VectorParameter() {parameterName=CustomColor_Rim, defaultValue=<rgba 1 1 1 1>}.RGB)
   B=Multiply(A=VectorParameter() {parameterName=CustomColor_Tire, defaultValue=<rgba 1 1 1 1>}.RGB, B=@n8.RGB)
   Alpha=@n9
  )
  n11 = TextureCoordinate()
  n12 = Add(
   A=Multiply(A=@n5, B=TextureSampleParameter2D() {parameterName=BC_Rim, texture=/VehicleCosmetics/Wheels/MaxisPsplus/Textures/T_Maxis_PSPlus_Rim_BC}.A)
   B=@n11
  )
  n13 = LinearInterpolate(
   A=TextureSampleParameter2D(UVs=StaticSwitchParameter(True=@n12, False=@n11) {parameterName=Eye}) {parameterName=BC, texture=&T_Maxis_PSPlus_Rim_BC}.RGB
   B=@n8.RGB
   Alpha=@n9
  )
  n14 = VectorParameter() {parameterName=Rim_Color_Base, defaultValue=<rgba 0.0281 0.0281 0.0281 1>}
  n15 = VectorParameter() {parameterName=Rim_Color_A, defaultValue=<rgba 0.0888 0.0375 0.0948 1>}
  n16 = TextureSampleParameter2D() {parameterName=Rim_Mask, samplerType=SAMPLERTYPE_LinearColor, texture=/VehicleCosmetics/Wheels/MaxisPsplus/Textures/T_Maxis_PSPlus_Rim_M}
  n17 = NamedRerouteDeclaration(None=@n16.R) {name=Rim_Color_A, nodeColor=<rgba 1 0.1875 0 1>}
  n18 = LinearInterpolate(
   A=LinearInterpolate(
    A=LinearInterpolate(A=@n14.RGB, B=@n15.RGB, Alpha=@n17)
    B=VectorParameter() {parameterName=Rim_Color_B, defaultValue=<rgba 0.104 0.5781 0 1>}.RGB
    Alpha=NamedRerouteDeclaration(None=@n16.G) {name=Rim_Color_B, nodeColor=<rgba 1 0.1875 0 1>}
   )
   B=VectorParameter() {parameterName=Rim_Color_C, defaultValue=<rgba 0.1073 0.1073 0.1073 1>}.RGB
   Alpha=NamedRerouteDeclaration(None=@n16.B) {name=Rim_Color_C, nodeColor=<rgba 1 0.1875 0 1>}
  )
  n19 = VectorParameter() {parameterName=Tire_Color_Base, defaultValue=<rgba 0.0281 0.0281 0.0281 1>}
  n20 = VectorParameter() {parameterName=Tire_Color_A, defaultValue=<rgba 0.4435 0.0046 0.4948 1>}
  n21 = TextureSampleParameter2D() {parameterName=Tire_Mask, samplerType=SAMPLERTYPE_LinearColor, texture=/VehicleCosmetics/Wheels/MaxisPsplus/Textures/T_Maxis_PSPlus_Tire_M}
  n22 = NamedRerouteDeclaration(None=@n21.R) {name=Tire_Color_A, nodeColor=<rgba 1 0.1875 0 1>}
  n23 = LinearInterpolate(
   A=LinearInterpolate(
    A=LinearInterpolate(A=@n19.RGB, B=@n20.RGB, Alpha=@n22)
    B=VectorParameter() {parameterName=Tire_Color_B, defaultValue=<rgba 0.3656 0.3656 0.3656 1>}.RGB
    Alpha=NamedRerouteDeclaration(None=@n21.G) {name=Tire_Color_B, nodeColor=<rgba 1 0.1875 0 1>}
   )
   B=VectorParameter() {parameterName=Tire_Color_C, defaultValue=<rgba 0 0.1073 0.003 1>}.RGB
   Alpha=NamedRerouteDeclaration(None=@n21.B) {name=Tire_Color_C, nodeColor=<rgba 1 0.1875 0 1>}
  )
  n24 = NamedRerouteUsage()
  n25 = ScalarParameter() {defaultValue=0.3, parameterName=Cell_Shadong_Power}
  n26 = NamedRerouteUsage()
  Multiply(
   A=StaticSwitchParameter(True=StaticSwitchParameter(True=@n10, False=@n13) {parameterName=Painted}, False=LinearInterpolate(A=@n18, B=@n23, Alpha=@n24)) {parameterName=Painted_Masks}
   B=Multiply(A=Max(A=@n25, B=@n26), B=ScalarParameter() {defaultValue=1, parameterName=Cell_Shading_Contrast})
  )
  NamedRerouteDeclaration(None=VertexColor().R) {name=VC_Red_CH, nodeColor=<rgba 1 0 0.0073 1>}
  n27 = TextureSampleParameter2D() {parameterName=Rim_Normal, samplerType=SAMPLERTYPE_Normal, texture=/VehicleCosmetics/Wheels/MaxisPsplus/Textures/T_Maxis_PSPlus_Rim_N}
  n28 = TextureSampleParameter2D() {parameterName=Tire_Normal, samplerType=SAMPLERTYPE_Normal, texture=/VehicleCosmetics/Wheels/MaxisPsplus/Textures/T_Maxis_PSPlus_Tire_N}
  n29 = NamedRerouteUsage()
  n30 = DotProduct(
   A=CameraVectorWS()
   B=StaticSwitchParameter(True=LinearInterpolate(A=@n27.RGB, B=@n28.RGB, Alpha=@n29), False=Constant3Vector()) {defaultValue=true, parameterName=Use_Normal}
  )
  n31 = ScalarParameter() {defaultValue=2, parameterName=StepCount}
  n32 = Ceil(None=Subtract(A=Multiply(A=@n30, B=@n31), B=ScalarParameter() {defaultValue=0.2, parameterName=StepAmount}))
  NamedRerouteDeclaration(None=Clamp(None=ConstantBiasScale(None=@n32) {bias=-0.4, scale=1})) {name=Cell_Shading, nodeColor=<rgba 0.2656 0 1 1>}
  n33 = Multiply(A=Time(), B=ScalarParameter() {defaultValue=0.1, parameterName=Speed})
  n34 = Sine(None=@n33) {period=2}
  n35 = Cosine(None=@n33) {period=2}
  n36 = Multiply(A=AppendVector(A=AppendVector(A=@n34, B=@n35), B=@n35), B=Constant3Vector() {constant=<rgba 1 1 1 0>})
  n37 = ScalarParameter() {defaultValue=1, parameterName=animation}
  NamedRerouteDeclaration(None=StaticSwitchParameter(True=Multiply(A=@n36, B=@n37), False=Constant()) {parameterName=Animation}) {name=WPO, nodeColor=<rgba 1 0.2109 0 1>}
  NamedRerouteUsage()
  n38 = VertexColor()
  n39 = Time()
  n40 = Constant() {r=3}
  n41 = Floor(None=ConstantBiasScale(None=Fmod(A=@n39, B=@n40)) {bias=0})
  StaticSwitchParameter(True=Add(A=@n38.B, B=LinearInterpolate(A=@n38.R, B=@n38.R, Alpha=@n41)), False=Constant() {r=1}) {parameterName="Animation_Facial_State(SwitchesBetween R and G, Blue is not effected)"}
 ]
}
